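# Post-process AnimBP for the SunnySquawk female medium head, adding secondary feather motion and a left-eyelid pose.
# A serial chain of spring-driven, angle-limited AnimDynamics nodes simulates left, right, back and neck feather bones with per-bone tuning.
# Additive offsets and a scale on the left upper and lower mid eyelids then blend in while bIsTargeting is true.
# Two duplicate back-feather nodes are disconnected from the output; the EventGraph holds only unwired stubs.
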AnimBlueprint /BRCosmetics/Characters/Player/Female/Medium/Heads/F_MED_SunnySquawk_Head/Meshes/F_MED_SunnySquawk_Head_Skeleton_PostProcessBP
target_skeleton=&/F_MED_SunnySquawk_Head_Skeleton
graphs [
 AnimGraph {nodes [
   n1 = Anim Dynamics - Bone: dyn_head_feather_back_1(Alpha=1, ExternalForce="0,0,0") {
    angularDampingOverride=0.9999
    angularSpringConstant=10000
    bAngularSpring=true
    bOverrideAngularDamping=true
    bUseGravityOverride=true
    gravityScale=0.35
    alphaBoolBlend {blendOption=Linear}
    boundBone {boneName=dyn_head_feather_back_1}
    componentAppliedLinearAccClamp=<1000 1000 1000>
    componentLinearAccScale=<1 1 1>
    componentLinearVelScale=<1 1 1>
    gravityOverride {z=-100}
    physicsBodyDefinitions [{
      boundBone {boneName=dyn_head_feather_back_1}
      boxExtents=<20 30 30>
      constraintSetup {angularLimitsMax=<0 12 2>, angularLimitsMin=<0 -12 -2>, angularTarget=<1 0 0>}
      localJointOffset=<4 0 0>
     }]
   }
   Anim Dynamics - Bone: dyn_head_feather_back_2(ComponentPose=@n1, Alpha=1, ExternalForce="0,0,0") {
    angularDampingOverride=0.9999
    angularSpringConstant=10000
    bAngularSpring=true
    bOverrideAngularDamping=true
    bUseGravityOverride=true
    gravityScale=0.35
    alphaBoolBlend {blendOption=Linear}
    boundBone {boneName=dyn_head_feather_back_2}
    componentAppliedLinearAccClamp=<1000 1000 1000>
    componentLinearAccScale=<1 1 1>
    componentLinearVelScale=<1 1 1>
    gravityOverride {z=-100}
    physicsBodyDefinitions [{
      boundBone {boneName=dyn_head_feather_back_2}
      boxExtents=<20 30 30>
      constraintSetup {angularLimitsMax=<0 15 2>, angularLimitsMin=<0 -15 -2>, angularTarget=<1 0 0>}
      localJointOffset=<4 0 0>
     }]
   }
   n2 = Get bIsTargeting()
   n3 = Anim Dynamics - Bone: dyn_head_feather_l_mstr(ComponentPose=Local To Component(LocalPose=Input Pose()), Alpha=1, ExternalForce="0,0,0") {
    angularDampingOverride=1
    angularSpringConstant=20000
    bAngularSpring=true
    bOverrideAngularDamping=true
    bUseGravityOverride=true
    gravityScale=0.35
    alphaBoolBlend {blendOption=Linear}
    boundBone {boneName=dyn_head_feather_l_mstr}
    componentLinearAccScale=<1 1 1>
    componentLinearVelScale=<1 1 1>
    gravityOverride {z=-980}
    physicsBodyDefinitions [{
      boundBone {boneName=dyn_head_feather_l_mstr}
      boxExtents=<20 30 30>
      constraintSetup {angularLimitsMax=<0 10 2>, angularLimitsMin=<0 -10 -2>, angularTarget=<1 0 0>}
      localJointOffset=<4 0 0>
     }]
   }
   n4 = Anim Dynamics - Bone: dyn_head_feather_l_1(ComponentPose=@n3, Alpha=1, ExternalForce="0,0,0") {
    angularDampingOverride=1
    angularSpringConstant=20000
    bAngularSpring=true
    bOverrideAngularDamping=true
    bUseGravityOverride=true
    gravityScale=0.35
    alphaBoolBlend {blendOption=Linear}
    boundBone {boneName=dyn_head_feather_l_1}
    componentLinearAccScale=<1.1 1.1 1.1>
    componentLinearVelScale=<1.1 1.1 1.1>
    gravityOverride {z=-980}
    physicsBodyDefinitions [{
      boundBone {boneName=dyn_head_feather_l_1}
      boxExtents=<18 18 18>
      constraintSetup {angularLimitsMax=<0 14 4>, angularLimitsMin=<0 -14 -4>, angularTarget=<1 0 0>}
      localJointOffset=<4 0 0>
     }]
   }
   n5 = Anim Dynamics - Bone: dyn_head_feather_l_2(ComponentPose=@n4, Alpha=1, ExternalForce="0,0,0") {
    angularDampingOverride=1
    angularSpringConstant=20000
    bAngularSpring=true
    bOverrideAngularDamping=true
    bUseGravityOverride=true
    gravityScale=0.35
    alphaBoolBlend {blendOption=Linear}
    boundBone {boneName=dyn_head_feather_l_2}
    componentLinearAccScale=<1 1 1>
    componentLinearVelScale=<1 1 1>
    gravityOverride {z=-980}
    physicsBodyDefinitions [{
      boundBone {boneName=dyn_head_feather_l_2}
      boxExtents=<20 35 35>
      constraintSetup {angularLimitsMax=<0 10 2>, angularLimitsMin=<0 -10 -2>, angularTarget=<1 0 0>}
      localJointOffset=<4 0 0>
     }]
   }
   n6 = Anim Dynamics - Bone: dyn_head_feather_l_3(ComponentPose=@n5, Alpha=1, ExternalForce="0,0,0") {
    angularDampingOverride=1
    angularSpringConstant=20000
    bAngularSpring=true
    bOverrideAngularDamping=true
    bUseGravityOverride=true
    gravityScale=0.35
    alphaBoolBlend {blendOption=Linear}
    boundBone {boneName=dyn_head_feather_l_3}
    componentLinearAccScale=<1 1 1>
    componentLinearVelScale=<1 1 1>
    gravityOverride {z=-980}
    physicsBodyDefinitions [{
      boundBone {boneName=dyn_head_feather_l_3}
      boxExtents=<19 33 33>
      constraintSetup {angularLimitsMax=<0 10 2>, angularLimitsMin=<0 -10 -2>, angularTarget=<1 0 0>}
      localJointOffset=<4 0 0>
     }]
   }
   n7 = Anim Dynamics - Bone: dyn_head_feather_r_mstr(ComponentPose=@n6, Alpha=1, ExternalForce="0,0,0") {
    angularDampingOverride=0.9999
    angularSpringConstant=10000
    bAngularSpring=true
    bOverrideAngularDamping=true
    bUseGravityOverride=true
    gravityScale=0.35
    alphaBoolBlend {blendOption=Linear}
    boundBone {boneName=dyn_head_feather_r_mstr}
    componentAppliedLinearAccClamp=<1000 1000 1000>
    componentLinearAccScale=<0.3 0.3 0.3>
    componentLinearVelScale=<0.7 0.7 0.7>
    gravityOverride {z=-100}
    physicsBodyDefinitions [{
      boundBone {boneName=dyn_head_feather_r_mstr}
      boxExtents=<25 35 35>
      constraintSetup {angularLimitsMax=<0 12 2>, angularLimitsMin=<0 -12 -2>, angularTarget=<1 0 0>}
      localJointOffset=<4 0 0>
     }]
   }
   n8 = Anim Dynamics - Bone: dyn_head_feather_r_1(ComponentPose=@n7, Alpha=1, ExternalForce="0,0,0") {
    angularDampingOverride=0.9999
    angularSpringConstant=10000
    bAngularSpring=true
    bOverrideAngularDamping=true
    bUseGravityOverride=true
    gravityScale=0.35
    alphaBoolBlend {blendOption=Linear}
    boundBone {boneName=dyn_head_feather_r_1}
    componentAppliedLinearAccClamp=<1000 1000 1000>
    componentLinearAccScale=<0.3 0.3 0.3>
    componentLinearVelScale=<0.7 0.7 0.7>
    gravityOverride {z=-100}
    physicsBodyDefinitions [{
      boundBone {boneName=dyn_head_feather_r_1}
      boxExtents=<25 35 35>
      constraintSetup {angularLimitsMax=<0 15 2>, angularLimitsMin=<0 -15 -2>, angularTarget=<1 0 0>}
      localJointOffset=<4 0 0>
     }]
   }
   n9 = Anim Dynamics - Bone: dyn_head_feather_r_2(ComponentPose=@n8, Alpha=1, ExternalForce="0,0,0") {
    angularDampingOverride=0.9999
    angularSpringConstant=10000
    bAngularSpring=true
    bOverrideAngularDamping=true
    bUseGravityOverride=true
    gravityScale=0.35
    alphaBoolBlend {blendOption=Linear}
    boundBone {boneName=dyn_head_feather_r_2}
    componentAppliedLinearAccClamp=<1000 1000 1000>
    componentLinearAccScale=<0.3 0.3 0.3>
    componentLinearVelScale=<0.7 0.7 0.7>
    gravityOverride {z=-100}
    physicsBodyDefinitions [{
      boundBone {boneName=dyn_head_feather_r_2}
      boxExtents=<25 35 35>
      constraintSetup {angularLimitsMax=<0 12 2>, angularLimitsMin=<0 -12 -2>, angularTarget=<1 0 0>}
      localJointOffset=<4 0 0>
     }]
   }
   n10 = Anim Dynamics - Bone: dyn_head_feather_r_3(ComponentPose=@n9, Alpha=1, ExternalForce="0,0,0") {
    angularDampingOverride=0.9999
    angularSpringConstant=10000
    bAngularSpring=true
    bOverrideAngularDamping=true
    bUseGravityOverride=true
    gravityScale=0.35
    alphaBoolBlend {blendOption=Linear}
    boundBone {boneName=dyn_head_feather_r_3}
    componentAppliedLinearAccClamp=<1000 1000 1000>
    componentLinearAccScale=<0.3 0.3 0.3>
    componentLinearVelScale=<0.7 0.7 0.7>
    gravityOverride {z=-100}
    physicsBodyDefinitions [{
      boundBone {boneName=dyn_head_feather_r_3}
      boxExtents=<25 35 35>
      constraintSetup {angularLimitsMax=<0 8 2>, angularLimitsMin=<0 -8 -2>, angularTarget=<1 0 0>}
      localJointOffset=<4 0 0>
     }]
   }
   n11 = Anim Dynamics - Bone: dyn_head_feather_back_1(ComponentPose=@n10, Alpha=1, ExternalForce="0,0,0") {
    angularDampingOverride=0.9999
    angularSpringConstant=20000
    bAngularSpring=true
    bOverrideAngularDamping=true
    bUseGravityOverride=true
    gravityScale=0.35
    alphaBoolBlend {blendOption=Linear}
    boundBone {boneName=dyn_head_feather_back_1}
    componentAppliedLinearAccClamp=<1500 1500 1500>
    componentLinearAccScale=<0.5 0.5 0.5>
    componentLinearVelScale=<0.6 0.6 0.6>
    gravityOverride {z=-800}
    physicsBodyDefinitions [{
      boundBone {boneName=dyn_head_feather_back_1}
      boxExtents=<15 30 30>
      constraintSetup {angularLimitsMax=<0 14 6>, angularLimitsMin=<0 -14 -6>, angularTarget=<1 0 0>}
      localJointOffset=<4 0 0>
     }]
   }
   n12 = Anim Dynamics - Bone: dyn_head_feather_back_2(ComponentPose=@n11, Alpha=1, ExternalForce="0,0,0") {
    angularDampingOverride=0.9999
    angularSpringConstant=20000
    bAngularSpring=true
    bOverrideAngularDamping=true
    bUseGravityOverride=true
    gravityScale=0.35
    alphaBoolBlend {blendOption=Linear}
    boundBone {boneName=dyn_head_feather_back_2}
    componentAppliedLinearAccClamp=<1500 1500 1500>
    componentLinearAccScale=<0.5 0.5 0.5>
    componentLinearVelScale=<0.6 0.6 0.6>
    gravityOverride {z=-800}
    physicsBodyDefinitions [{
      boundBone {boneName=dyn_head_feather_back_2}
      boxExtents=<25 35 35>
      constraintSetup {angularLimitsMax=<0 14 6>, angularLimitsMin=<0 -14 -6>, angularTarget=<1 0 0>}
      localJointOffset=<4 0 0>
     }]
   }
   n13 = Anim Dynamics - Bone: dyn_feather_neck_1(ComponentPose=@n12, Alpha=1, ExternalForce="0,0,0") {
    angularDampingOverride=1
    angularSpringConstant=20000
    bAngularSpring=true
    bOverrideAngularDamping=true
    bUseGravityOverride=true
    gravityScale=0.35
    alphaBoolBlend {blendOption=Linear}
    boundBone {boneName=dyn_feather_neck_1}
    componentAppliedLinearAccClamp=<1000 1000 1000>
    componentLinearAccScale=<1 1 1>
    componentLinearVelScale=<1 1 1>
    gravityOverride {z=-980}
    physicsBodyDefinitions [{
      boundBone {boneName=dyn_feather_neck_1}
      boxExtents=<24 35 35>
      constraintSetup {angularLimitsMax=<0 15 2>, angularLimitsMin=<0 -25 -2>, angularTarget=<1 0 0>}
      localJointOffset=<4 0 0>
     }]
   }
   n14 = Anim Dynamics - Bone: dyn_feather_neck_2(ComponentPose=@n13, Alpha=1, ExternalForce="0,0,0") {
    angularDampingOverride=1
    angularSpringConstant=20000
    bAngularSpring=true
    bOverrideAngularDamping=true
    bUseGravityOverride=true
    gravityScale=0.35
    alphaBoolBlend {blendOption=Linear}
    boundBone {boneName=dyn_feather_neck_2}
    componentAppliedLinearAccClamp=<1000 1000 1000>
    componentLinearAccScale=<1.2 1.2 1.2>
    componentLinearVelScale=<1.3 1.3 1.3>
    gravityOverride {z=-980}
    physicsBodyDefinitions [{
      boundBone {boneName=dyn_feather_neck_2}
      boxExtents=<30 31 33>
      constraintSetup {angularLimitsMax=<0 14 4>, angularLimitsMin=<0 -30 -4>, angularTarget=<1 0 0>}
      localJointOffset=<4 0 0>
     }]
   }
   n15 = Anim Dynamics - Bone: dyn_feather_neck_3(ComponentPose=@n14, Alpha=1, ExternalForce="0,0,0") {
    angularDampingOverride=0.9999
    angularSpringConstant=10000
    bAngularSpring=true
    bOverrideAngularDamping=true
    bUseGravityOverride=true
    gravityScale=0.35
    alphaBoolBlend {blendOption=Linear}
    boundBone {boneName=dyn_feather_neck_3}
    componentAppliedLinearAccClamp=<1000 1000 1000>
    componentLinearAccScale=<0.4 0.3 0.4>
    componentLinearVelScale=<0.7 0.7 0.7>
    gravityOverride {z=-100}
    physicsBodyDefinitions [{
      boundBone {boneName=dyn_feather_neck_3}
      boxExtents=<25 35 35>
      constraintSetup {angularLimitsMax=<0 17 2>, angularLimitsMin=<0 -25 -2>, angularTarget=<1 0 0>}
      localJointOffset=<4 0 0>
     }]
   }
   n16 = Anim Dynamics - Bone: dyn_feather_neck_4(ComponentPose=@n15, Alpha=1, ExternalForce="0,0,0") {
    angularDampingOverride=0.9999
    angularSpringConstant=10000
    bAngularSpring=true
    bOverrideAngularDamping=true
    bUseGravityOverride=true
    gravityScale=0.35
    alphaBoolBlend {blendOption=Linear}
    boundBone {boneName=dyn_feather_neck_4}
    componentAppliedLinearAccClamp=<1000 1000 1000>
    componentLinearAccScale=<0.4 0.4 0.5>
    componentLinearVelScale=<0.7 0.7 0.7>
    gravityOverride {z=-100}
    physicsBodyDefinitions [{
      boundBone {boneName=dyn_feather_neck_4}
      boxExtents=<22 33 33>
      constraintSetup {angularLimitsMax=<0 17 2>, angularLimitsMin=<0 -25 -2>, angularTarget=<1 0 0>}
      localJointOffset=<4 0 0>
     }]
   }
   n17 = "Transform (Modify) Bone - Bone: L_eye_lid_upper_mid"!(ComponentPose=@n16, bAlphaBoolEnabled:bool=@n2, Translation="-0.05,0,-0.08", Rotation="-4.679,0.000,-6.297", Scale="1.05,1.05,1.05") -> (Pose) {
    alphaInputType=Bool
    rotationMode=BMM_Additive
    rotationSpace=BCS_BoneSpace
    scaleMode=BMM_Replace
    translationMode=BMM_Additive
    translationSpace=BCS_BoneSpace
    alphaBoolBlend {blendInTime=0.1, blendOption=Linear, blendOutTime=0.1}
    boneToModify {boneName=L_eye_lid_upper_mid}
    rotation=<rot -4.679 0 -6.2969>
    scale=<1.05 1.05 1.05>
    translation {x=-0.05, z=-0.08}
   }
   n18 = "Transform (Modify) Bone - Bone: L_eye_lid_lower_mid"!(ComponentPose=@n17, bAlphaBoolEnabled:bool=@n2, Translation="0,0,-0.1", Rotation="5.455,-0.031,-5.333", Scale="1.05,1.05,1.05") -> (Pose) {
    alphaInputType=Bool
    rotationMode=BMM_Additive
    rotationSpace=BCS_BoneSpace
    scaleMode=BMM_Replace
    translationMode=BMM_Additive
    translationSpace=BCS_BoneSpace
    alphaBoolBlend {blendInTime=0.1, blendOption=Linear, blendOutTime=0.1}
    boneToModify {boneName=L_eye_lid_lower_mid}
    rotation=<rot 5.455 -0.031 -5.333>
    scale=<1.05 1.05 1.05>
    translation {z=-0.1}
   }
   Output Pose(Result=Component To Local(ComponentPose=@n18))
  ]}
 EventGraph {nodes [
   TryGetPawnOwner()
   Event BlueprintUpdateAnimation()
  ]}
]
variables []
